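# Niagara system for the HoloGlass vehicle boost cosmetic: glass-like mesh shards and colored sprites around the boost position.
# Sprites streams them backward and emits death events, while Sprites-LocalSpace adds glow, core and wobbling sprite layers, both merging layers via spawn groups.
# User flags switch random holographic palettes to painted colors and adjust trail lifetime and speed for frontend preview.
# FakeGlow and TriangleExplosion002 have their spawn rates disabled, with no other spawn source shown.
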NiagaraSystem /VehicleCosmetics/Boosts/Boost_HoloGlass/FX/NS_Boost_HoloGlass_MP
dependencies {
 usedDataInterfaces [&NiagaraDataInterfaceCurve, &NiagaraDataInterfaceArrayColor, &NiagaraDataInterfaceSkeletalMesh]
 usedDynamicInputs []
 usedModules [
  &self:SystemSpawnScript.SystemScriptSource.SystemScriptGraph.NiagaraNodeAssignment_0.SetVariables_#1
  &SystemState
 ]
}
emitters {
 FakeGlow {
  emitterUpdateScript [
   &EmitterState()
   &VC_FilteredSocketCount() {enabled=false}
   &SpawnRate(SpawnRate=50, Use Spawn Probability=false) {enabled=false}
  ]
  particleSpawnScript [
   &VC_PaintedVariantsOverride() {enabled=false}
   &InitializeParticle(
    Color Mode=::NewEnumerator0
    Lifetime=0.1
    Sprite Rotation Mode=::NewEnumerator2
    Sprite Size Mode=::NewEnumerator3
    Uniform Sprite Size=200
    UsePositionOffset=false
    Write Material Random=true
   )
   &PartitionParticles(Number Of Partitions=$Emitter.PortCount) {enabled=false}
   &SkeletalMeshLocation(
    Bone / Socket Index=$Particles.PartitionParticles.Partition
    Bone / Socket Sampling Mode=::NewEnumerator5
    MaskSpawnGroup=false
    Mesh Sampling Type=::NewEnumerator6
    Sampled Position Offset=<-60 0 0>
    Skeletal Mesh=$System.SkeletalMesh
   )
   &AddVelocity(Velocity=<-1500 0 0>) {enabled=false}
   &Color(Color=<rgba 1 1 1 0.03>)
  ]
  particleUpdateScript [
   &ParticleState()
   &ScaleColor(
    Scale Alpha=&MakeCustomFloatFromBool(
     Bool=$User.bIsPainted
     True Float=0
     False Float=&FloatFromCurve(FloatCurve=DataInterface() {propertyValues {Curve {keys [<at 0 0 RCIM_Cubic>,<at 0.5 1 RCIM_Cubic>,<at 1 0 RCIM_Cubic>]}}})
    )
    ScaleA=true
    ScaleRGB=true
   )
   &SolveForcesAndVelocity(Clamp Velocity=false, Limit Acceleration=false)
  ]
  rendererClasses [&NiagaraSpriteRendererProperties]
  renderers [{rendererClass=&NiagaraSpriteRendererProperties}]
 }
 Sprites {
  emitterUpdateScript [
   &EmitterState()
   &VC_FilteredSocketCount() {enabled=false}
   &SpawnRate(SpawnRate=50, Use Spawn Probability=false)
   &SpawnRate(SpawnGroup=1, SpawnRate=50, Use Spawn Probability=false) {module=SpawnRate001}
   &SpawnRate(SpawnGroup=2, SpawnRate=50, Use Spawn Probability=false) {module=SpawnRate002}
  ]
  particleSpawnScript [
   /SetVariables(Particles.Vistag:NiagaraInt32=&SpawnGroup()) {isSetParameters=true}
   &VC_PaintedVariantsOverride() {enabled=false}
   &InitializeParticle(
    Color Mode=::NewEnumerator0
    Lifetime=/VehicleCosmetics/FX/Niagara/Modules/MergeEmitters/VC_FloatBySpawnGroup(
     SpawnGroup or ExecIndex:NiagaraInt32=0
     CustomInt:NiagaraInt32=0
     MaximumSpawnGroups:NiagaraInt32=3
     Float SG 0:NiagaraFloat=&MakeCustomFloatFromBool(
      Bool=$User.bIsFrontendPreview
      True Float=&RandomRangeFloat(Minimum=0.1, Maximum=0.2)
      False Float=&RandomRangeFloat(Minimum=0.07, Maximum=0.15)
     )
     Float SG 1:NiagaraFloat=&MakeCustomFloatFromBool(
      Bool=$User.bIsFrontendPreview
      True Float=&RandomRangeFloat(Minimum=0.4, Maximum=0.6)
      False Float=&RandomRangeFloat(Minimum=0.1, Maximum=0.2)
     )
     Float SG 2:NiagaraFloat=&MakeCustomFloatFromBool(
      Bool=$User.bIsFrontendPreview
      True Float=&RandomRangeFloat(Minimum=0.4, Maximum=0.6)
      False Float=&RandomRangeFloat(Minimum=0.1, Maximum=0.2)
     )
     Float SG 3:NiagaraFloat=0
     Float SG 4:NiagaraFloat=0
     Float SG 5:NiagaraFloat=0
    )
    Mesh Scale=&Multiply_Vector(
     A=&VectorFromFloat(Value=0.2)
     B=&RandomRangeVector(Minimum=<1 1 1>, Maximum=<1 3 1>)
    )
    Mesh Scale Mode=::NewEnumerator4
    Position Offset=&AddVectorToPosition(
     Position=$System.FinalPosition
     Vector=/VehicleCosmetics/FX/Niagara/Modules/MergeEmitters/VC_VectorBySpawnGroup(
      SpawnGroup or ExecIndex:NiagaraInt32=0
      CustomInt:NiagaraInt32=0
      MaximumSpawnGroups:NiagaraInt32=3
      Vector Group 0:Vector3f=<-60 0 0>
      Vector Group 1:Vector3f=<-70 0 0>
      Vector Group 2:Vector3f=<-70 0 0>
      Vector Group 3:Vector3f=<0 0 0>
      Vector Group 4:Vector3f=<0 0 0>
      Vector Group 5:Vector3f=<0 0 0>
     )
    )
    Position Offset Coordinate Space=::Local
    Sprite Rotation Mode=::NewEnumerator2
    Sprite Size Mode=::NewEnumerator3
    Uniform Sprite Size=&VC_FloatBySpawnGroup(
     SpawnGroup or ExecIndex:NiagaraInt32=0
     CustomInt:NiagaraInt32=0
     MaximumSpawnGroups:NiagaraInt32=3
     Float SG 0:NiagaraFloat=0
     Float SG 1:NiagaraFloat=&Multiply_Float(A=0.2, B=&RandomRangeFloat(Minimum=20, Maximum=50))
     Float SG 2:NiagaraFloat=&Multiply_Float(A=0.3, B=&RandomRangeFloat(Minimum=20, Maximum=50))
     Float SG 3:NiagaraFloat=0
     Float SG 4:NiagaraFloat=0
     Float SG 5:NiagaraFloat=0
    )
    UsePositionOffset=true
    Write Material Random=true
   )
   &PartitionParticles(Number Of Partitions=$Emitter.PortCount) {enabled=false}
   &SkeletalMeshLocation(
    Bone / Socket Index=$Particles.PartitionParticles.Partition
    Bone / Socket Sampling Mode=::NewEnumerator5
    MaskSpawnGroup=false
    Mesh Sampling Type=::NewEnumerator6
    Sampled Position Offset=<-90 -35 0>
    Skeletal Mesh=$System.SkeletalMesh
   ) {enabled=false}
   &AddVelocity(Velocity=&VC_VectorBySpawnGroup(
     SpawnGroup or ExecIndex:NiagaraInt32=0
     CustomInt:NiagaraInt32=0
     MaximumSpawnGroups:NiagaraInt32=3
     Vector Group 0:Vector3f=&MakeCustomVectorFromBool(Bool=$User.bIsFrontendPreview, True Vector=<-3000 0 0>, False Vector=<-1000 0 0>)
     Vector Group 1:Vector3f=&MakeCustomVectorFromBool(Bool=$User.bIsFrontendPreview, True Vector=<-1500 0 0>, False Vector=<-500 0 0>)
     Vector Group 2:Vector3f=&MakeCustomVectorFromBool(Bool=$User.bIsFrontendPreview, True Vector=<-1500 0 0>, False Vector=<-500 0 0>)
     Vector Group 3:Vector3f=<0 0 0>
     Vector Group 4:Vector3f=<0 0 0>
     Vector Group 5:Vector3f=<0 0 0>
    ))
   &ShapeLocation(MaskSpawnGroup=false, Override Seed=false, Sphere Radius=30, UseOwnerScale=true)
   &Color(Color=/VehicleCosmetics/FX/Niagara/Modules/MergeEmitters/VC_Dynamic_ColorBySpawnGroup(
     Color SG 0:LinearColor=&MakeCustomLinearColorFromBool(
      Bool=$User.bIsPainted
      True Color=&Multiply_LinearColorByFloat(
       Linear Color=$User.GlassColor
       Float=&RandomRangeFloat(Minimum=$User.MinimumGlassColor, Maximum=$User.MaxGlassColor)
      )
      False Color=&SelectLinearColorFromArray(
       Array Sampling Mode=::NewEnumerator0
       Direct Array Index=?
       Color Selection Array=DataInterface() {propertyValues {ColorData [<rgba 1 0.01 1 1>, <rgba 0.01 1 1 1>, <rgba 2 1 0.02 1>, <rgba 0.5073 0.1771 1 1>]}}
       Random Seed=0
      )
     )
     MaximumSpawnGroups:NiagaraInt32=3
     Color SG 1:LinearColor=&MakeCustomLinearColorFromBool(
      Bool=$User.bIsPainted
      True Color=$User.PaintedColor
      False Color=&SelectLinearColorFromArray(
       Array Sampling Mode=::NewEnumerator0
       Direct Array Index=?
       Color Selection Array=DataInterface() {propertyValues {ColorData [<rgba 1 0.2 1 1>, <rgba 0.2 1 1 1>, <rgba 1 1 0.05 1>, <rgba 0.2 0.05 1 1>]}}
       Random Seed=0
      )
     )
     Color SG 2:LinearColor=&MakeCustomLinearColorFromBool(
      Bool=$User.bIsPainted
      True Color=$User.PaintedColor
      False Color=&SelectLinearColorFromArray(
       Array Sampling Mode=::NewEnumerator0
       Direct Array Index=?
       Color Selection Array=DataInterface() {propertyValues {ColorData [<rgba 1 0.01 1 1>, <rgba 0.01 1 1 1>, <rgba 1 0.5 0.01 1>, <rgba 0.2 0.01 1 1>]}}
       Random Seed=0
      )
     )
     Color SG 3:LinearColor=<rgba 1 1 1 1>
     Color SG 4:LinearColor=<rgba 1 1 1 1>
    ))
   &InitialMeshOrientation(
    Mesh Axis To Orient=<0 0 0>
    Mesh Orientation Mode=::NewEnumerator2
    Rotation=&RandomRangeVector(Minimum=<-1 0 0>, Maximum=<1 0 0>)
    Use Rotation Vector=false
   )
   /SetVariables(Particles.MeshIndex=&RandomRangeInt(Minimum=0, Maximum=3)) {isSetParameters=true}
  ]
  particleUpdateScript [
   &ParticleState()
   &SpriteRotationRate(Rotation Rate=&VC_FloatBySpawnGroup(
     SpawnGroup or ExecIndex:NiagaraInt32=0
     CustomInt:NiagaraInt32=0
     MaximumSpawnGroups:NiagaraInt32=3
     Float SG 0:NiagaraFloat=0
     Float SG 1:NiagaraFloat=0
     Float SG 2:NiagaraFloat=&RandomRangeFloat(Minimum=500, Maximum=-500)
     Float SG 3:NiagaraFloat=0
     Float SG 4:NiagaraFloat=0
     Float SG 5:NiagaraFloat=0
    ))
   &ScaleColor(
    Scale Alpha=&VC_FloatBySpawnGroup(
     SpawnGroup or ExecIndex:NiagaraInt32=0
     CustomInt:NiagaraInt32=0
     MaximumSpawnGroups:NiagaraInt32=3
     Float SG 0:NiagaraFloat=1
     Float SG 1:NiagaraFloat=1
     Float SG 2:NiagaraFloat=1
     Float SG 3:NiagaraFloat=1
     Float SG 4:NiagaraFloat=0
     Float SG 5:NiagaraFloat=0
    )
    Scale RGB=&VC_VectorBySpawnGroup(
     SpawnGroup or ExecIndex:NiagaraInt32=0
     CustomInt:NiagaraInt32=0
     MaximumSpawnGroups:NiagaraInt32=3
     Vector Group 0:Vector3f=&MakeCustomVectorFromBool(
      Bool=$User.bIsPainted
      True Vector=&VectorFromFloat(Value=&Multiply_Float(
        A=&RandomRangeFloat(Minimum=$User.GlassColorVar, Maximum=1)
        B=&FloatFromCurve(FloatCurve=DataInterface() {propertyValues {Curve {keys [<at 0 1>,<at 0.9 1>,<at 1 2>]}}})
       ))
      False Vector=&VectorFromFloat(Value=&RandomRangeFloat(
        Minimum=&FloatFromCurve(FloatCurve=DataInterface() {propertyValues {Curve {keys [<at 0 10>,<at 0.1 1>,<at 0.9 1>,<at 1 10>]}}}, Scale Curve=0.5)
        Maximum=&FloatFromCurve(FloatCurve=DataInterface() {propertyValues {Curve {keys [<at 0 5>,<at 0.1 1>,<at 0.9 1>,<at 1 5>]}}}, Scale Curve=0.6)
       ))
     )
     Vector Group 1:Vector3f=&MakeCustomVectorFromBool(
      Bool=$User.bIsPainted
      True Vector=&VectorFromFloat(Value=&RandomRangeFloat(Minimum=$User.ParticleColorVar, Maximum=1))
      False Vector=&Multiply_VectorByFloat(
       Vector=&VectorFromFloat(Value=&RandomRangeFloat(Minimum=1, Maximum=3))
       Float=0.8
      )
     )
     Vector Group 2:Vector3f=&MakeCustomVectorFromBool(
      Bool=$User.bIsPainted
      True Vector=&VectorFromFloat(Value=&RandomRangeFloat(Minimum=$User.ParticleColorVar, Maximum=1))
      False Vector=&Multiply_VectorByFloat(
       Vector=&VectorFromFloat(Value=&RandomRangeFloat(Minimum=1, Maximum=2.5))
       Float=0.8
      )
     )
     Vector Group 3:Vector3f=<0 0 0>
     Vector Group 4:Vector3f=<0 0 0>
     Vector Group 5:Vector3f=<0 0 0>
    )
    ScaleA=true
    ScaleRGB=true
   )
   &SolveForcesAndVelocity(Clamp Velocity=false, Limit Acceleration=false)
   &GenerateDeathEvent(Use Event Probability=false)
  ]
  rendererClasses [&NiagaraMeshRendererProperties, &NiagaraSpriteRendererProperties]
  renderers [
   {rendererClass=&NiagaraMeshRendererProperties, rendererIndex=0}
   {rendererClass=&NiagaraSpriteRendererProperties, rendererIndex=1}
   {rendererClass=&NiagaraSpriteRendererProperties, rendererIndex=2}
   {rendererClass=&NiagaraSpriteRendererProperties, rendererIndex=3}
  ]
 }
 Sprites-LocalSpace {
  emitterUpdateScript [
   &EmitterState()
   &VC_FilteredSocketCount() {enabled=false}
   &SpawnRate(SpawnRate=16, Use Spawn Probability=false)
   &SpawnRate(SpawnGroup=1, SpawnRate=6, Use Spawn Probability=false) {module=SpawnRate001}
   &SpawnRate(SpawnGroup=2, SpawnRate=16, Use Spawn Probability=false) {module=SpawnRate002}
   &SpawnRate(SpawnGroup=3, SpawnRate=60, Use Spawn Probability=false) {module=SpawnRate003}
   &SpawnRate(SpawnGroup=4, SpawnRate=40, Use Spawn Probability=false) {module=SpawnRate004}
  ]
  particleSpawnScript [
   /SetVariables(Particles.Vistag:NiagaraInt32=&SpawnGroup()) {isSetParameters=true}
   /SetVariables(Particles.MeshIndex=&RandomRangeInt(Minimum=0, Maximum=3)) {isSetParameters=true}
   &VC_PaintedVariantsOverride() {enabled=false}
   &InitializeParticle(
    Color Mode=::NewEnumerator0
    Lifetime=&VC_FloatBySpawnGroup(
     SpawnGroup or ExecIndex:NiagaraInt32=0
     CustomInt:NiagaraInt32=0
     MaximumSpawnGroups:NiagaraInt32=5
     Float SG 0:NiagaraFloat=&RandomRangeFloat(Minimum=0.4, Maximum=0.6)
     Float SG 1:NiagaraFloat=0.5
     Float SG 2:NiagaraFloat=&RandomRangeFloat(Minimum=0.4, Maximum=0.6)
     Float SG 3:NiagaraFloat=&RandomRangeFloat(Minimum=0.4, Maximum=0.6)
     Float SG 4:NiagaraFloat=&RandomRangeFloat(Minimum=0.7, Maximum=0.5)
     Float SG 5:NiagaraFloat=0
    )
    Mesh Scale=&Multiply_Vector(
     A=&VectorFromFloat(Value=0.2)
     B=&RandomRangeVector(Minimum=<1 2 1>, Maximum=<0.5 3 0.5>)
    )
    Mesh Scale Mode=::NewEnumerator4
    Position Offset=&AddVectorToPosition(Position=$System.FinalPosition, Vector=<-100 0 0>)
    Sprite Rotation Angle Max:NiagaraFloat=&VC_FloatBySpawnGroup(
     SpawnGroup or ExecIndex:NiagaraInt32=0
     CustomInt:NiagaraInt32=0
     MaximumSpawnGroups:NiagaraInt32=5
     Float SG 0:NiagaraFloat=360
     Float SG 1:NiagaraFloat=0
     Float SG 2:NiagaraFloat=360
     Float SG 3:NiagaraFloat=0
     Float SG 4:NiagaraFloat=0
     Float SG 5:NiagaraFloat=0
    )
    Sprite Rotation Angle Min:NiagaraFloat=&VC_FloatBySpawnGroup(
     SpawnGroup or ExecIndex:NiagaraInt32=0
     CustomInt:NiagaraInt32=0
     MaximumSpawnGroups:NiagaraInt32=5
     Float SG 0:NiagaraFloat=0
     Float SG 1:NiagaraFloat=0
     Float SG 2:NiagaraFloat=0
     Float SG 3:NiagaraFloat=0
     Float SG 4:NiagaraFloat=0
     Float SG 5:NiagaraFloat=0
    )
    Sprite Rotation Mode=::NewEnumerator2
    Sprite Size Mode=::NewEnumerator3
    Uniform Sprite Size=&VC_FloatBySpawnGroup(
     SpawnGroup or ExecIndex:NiagaraInt32=0
     CustomInt:NiagaraInt32=0
     MaximumSpawnGroups:NiagaraInt32=5
     Float SG 0:NiagaraFloat=&Multiply_Float(A=4, B=&RandomRangeFloat(Minimum=20, Maximum=50))
     Float SG 1:NiagaraFloat=12
     Float SG 2:NiagaraFloat=&Multiply_Float(A=0.5, B=&RandomRangeFloat(Minimum=20, Maximum=50))
     Float SG 3:NiagaraFloat=&Multiply_Float(A=0.2, B=&RandomRangeFloat(Minimum=20, Maximum=50))
     Float SG 4:NiagaraFloat=&RandomRangeFloat(Minimum=0.7, Maximum=0.5)
     Float SG 5:NiagaraFloat=0
    )
    UsePositionOffset=true
    Write Material Random=true
   )
   &PartitionParticles(Number Of Partitions=$Emitter.PortCount) {enabled=false}
   &SkeletalMeshLocation(
    Bone / Socket Index=$Particles.PartitionParticles.Partition
    Bone / Socket Sampling Mode=::NewEnumerator5
    MaskSpawnGroup=false
    Mesh Sampling Type=::NewEnumerator6
    Skeletal Mesh=$System.SkeletalMesh
   ) {enabled=false}
   &AddVelocity(Velocity=<-1500 0 0>) {enabled=false}
   &Color(Color=&VC_Dynamic_ColorBySpawnGroup(
     Color SG 0:LinearColor=&MakeCustomLinearColorFromBool(
      Bool=$User.bIsPainted
      True Color=$User.GlowColor
      False Color=&SelectLinearColorFromArray(
       Array Sampling Mode=::NewEnumerator0
       Direct Array Index=?
       Color Selection Array=DataInterface() {propertyValues {ColorData [<rgba 1 0.2 1 1>, <rgba 0.2 1 1 1>, <rgba 1 1 0.05 1>, <rgba 0.2 0.05 1 1>]}}
       Random Seed=0
      )
     )
     MaximumSpawnGroups:NiagaraInt32=5
     Color SG 1:LinearColor=&MakeCustomLinearColorFromBool(Bool=$User.bIsPainted, True Color=$User.CoreColor, False Color=<rgba 1 0.3 0.01 1>)
     Color SG 2:LinearColor=&MakeCustomLinearColorFromBool(
      Bool=$User.bIsPainted
      True Color=$User.GlowColor
      False Color=&SelectLinearColorFromArray(
       Array Sampling Mode=::NewEnumerator0
       Direct Array Index=?
       Color Selection Array=DataInterface() {propertyValues {ColorData [<rgba 1 0.2 1 1>, <rgba 0.2 1 1 1>, <rgba 1 1 0.05 1>, <rgba 0.2 0.05 1 1>]}}
       Random Seed=0
      )
     )
     Color SG 3:LinearColor=&MakeCustomLinearColorFromBool(
      Bool=$User.bIsPainted
      True Color=$User.PaintedColor
      False Color=&SelectLinearColorFromArray(
       Array Sampling Mode=::NewEnumerator0
       Direct Array Index=?
       Color Selection Array=DataInterface() {propertyValues {ColorData [<rgba 1 0.2 1 1>, <rgba 0.2 1 1 1>, <rgba 1 1 0.05 1>, <rgba 0.2 0.05 1 1>]}}
       Random Seed=0
      )
     )
     Color SG 4:LinearColor=&MakeCustomLinearColorFromBool(
      Bool=$User.bIsPainted
      True Color=&Multiply_LinearColorByFloat(
       Linear Color=$User.GlassColor
       Float=&RandomRangeFloat(Minimum=$User.MinimumGlassColor, Maximum=$User.MaxGlassColor)
      )
      False Color=&SelectLinearColorFromArray(
       Array Sampling Mode=::NewEnumerator0
       Direct Array Index=?
       Color Selection Array=DataInterface() {propertyValues {ColorData [<rgba 1 0.01 1 1>, <rgba 0.015 1.5 1.5 1>, <rgba 2 1.3621 0.0974 1>, <rgba 0.6862 0.1406 1.5 5>]}}
       Random Seed=0
      )
     )
    ))
   &InitialMeshOrientation(Use Rotation Vector=false)
   &ShapeLocation(
    MaskSpawnGroup=false
    Override Seed=false
    Sphere Radius=&MaskFloatBySpawnGroup(Masked Float=5, Passthrough Float=0)
    UseOwnerScale=true
   )
  ]
  particleUpdateScript [
   &ParticleState()
   &ScaleSpriteSize(
    Scale Sprite Size Mode=::NewEnumerator0
    Uniform Scale Factor=&VC_FloatBySpawnGroup(
     SpawnGroup or ExecIndex:NiagaraInt32=0
     CustomInt:NiagaraInt32=0
     MaximumSpawnGroups:NiagaraInt32=5
     Float SG 0:NiagaraFloat=&FloatFromCurve(FloatCurve=DataInterface() {propertyValues {Curve {keys [<at 0 0>,<at 0.5 1>,<at 1 0>]}}})
     Float SG 1:NiagaraFloat=&FloatFromCurve(FloatCurve=DataInterface() {propertyValues {Curve {keys [{tangentMode=RCTM_User, time=0.2, value=1}, {tangentMode=RCTM_User, time=1, value=8}]}}})
     Float SG 2:NiagaraFloat=&FloatFromCurve(FloatCurve=DataInterface() {propertyValues {Curve {keys [<at 0 0>,<at 0.5 1>,<at 1 0>]}}})
     Float SG 3:NiagaraFloat=&FloatFromCurve(FloatCurve=DataInterface() {propertyValues {Curve {keys [{interpMode=RCIM_Cubic, tangentMode=RCTM_User, time=0.4, value=1}, {interpMode=RCIM_Cubic, tangentMode=RCTM_User, time=1, value=0}]}}})
     Float SG 4:NiagaraFloat=0
     Float SG 5:NiagaraFloat=0
    )
   )
   &ScaleMeshSize(Scale Factor=&VectorFromFloat(Value=&FloatFromCurve(FloatCurve=DataInterface() {propertyValues {Curve {keys [
          <at 0 0 RCIM_Cubic>
          {arriveTangentWeight=0.232, interpMode=RCIM_Cubic, leaveTangentWeight=0.116, tangentMode=RCTM_User, time=0.4, value=1}
          {arriveTangentWeight=0.0534, interpMode=RCIM_Cubic, leaveTangentWeight=0.3735, tangentMode=RCTM_User, time=0.7, value=1}
          <at 1 0 RCIM_Cubic>
         ]}}})))
   &UpdateMeshOrientation(Rotation Rate=0.5, Rotation Vector=&RandomRangeVector(Minimum=<-1 -1 -1>, Maximum=<1 1 1>))
   &ScaleColor(
    Scale Alpha=&VC_FloatBySpawnGroup(
     SpawnGroup or ExecIndex:NiagaraInt32=0
     CustomInt:NiagaraInt32=0
     MaximumSpawnGroups:NiagaraInt32=5
     Float SG 0:NiagaraFloat=0.1
     Float SG 1:NiagaraFloat=&FloatFromCurve(FloatCurve=DataInterface() {propertyValues {Curve {keys [<at 0 0.5>,<at 1 0>]}}})
     Float SG 2:NiagaraFloat=1
     Float SG 3:NiagaraFloat=&FloatFromCurve(FloatCurve=DataInterface() {propertyValues {Curve {keys [<at 0 0>,<at 0.3 1>,<at 0.7 1>,<at 1 0>]}}})
     Float SG 4:NiagaraFloat=1
     Float SG 5:NiagaraFloat=0
    )
    Scale RGB=&VC_VectorBySpawnGroup(
     SpawnGroup or ExecIndex:NiagaraInt32=0
     CustomInt:NiagaraInt32=0
     MaximumSpawnGroups:NiagaraInt32=5
     Vector Group 0:Vector3f=&MakeCustomVectorFromBool(
      Bool=$User.bIsPainted
      True Vector=&VectorFromFloat(Value=&FloatFromCurve(FloatCurve=DataInterface() {propertyValues {Curve {keys [{tangentMode=RCTM_User, time=0, value=3}, {tangentMode=RCTM_User, time=1, value=1}]}}}))
      False Vector=&VectorFromFloat(Value=&FloatFromCurve(FloatCurve=DataInterface() {propertyValues {Curve {keys [<at 0 3>,<at 1 0>]}}}))
     )
     Vector Group 1:Vector3f=&VectorFromFloat(Value=&FloatFromCurve(FloatCurve=DataInterface() {propertyValues {Curve {keys [<at 0 10>,<at 1 2>]}}}))
     Vector Group 2:Vector3f=&MakeCustomVectorFromBool(
      Bool=$User.bIsPainted
      True Vector=&VectorFromFloat(Value=&FloatFromCurve(
        FloatCurve=DataInterface() {propertyValues {Curve {keys [{tangentMode=RCTM_User, time=0, value=5}, {tangentMode=RCTM_User, time=1, value=1}]}}}
        Scale Curve=15
       ))
      False Vector=&VectorFromFloat(Value=&FloatFromCurve(FloatCurve=DataInterface() {propertyValues {Curve {keys [<at 0 5>,<at 1 0>]}}}, Scale Curve=15))
     )
     Vector Group 3:Vector3f=&MakeCustomVectorFromBool(
      Bool=$User.bIsPainted
      True Vector=&VectorFromFloat(Value=&RandomRangeFloat(Minimum=$User.ParticleColorVar, Maximum=1))
      False Vector=&Multiply_VectorByFloat(
       Vector=&VectorFromFloat(Value=&FloatFromCurve(FloatCurve=DataInterface() {propertyValues {Curve {keys [<at 0 20>,<at 0.5 1>]}}}))
       Float=1
      )
     )
     Vector Group 4:Vector3f=&MakeCustomVectorFromBool(
      Bool=$User.bIsPainted
      True Vector=&VectorFromFloat(Value=&Multiply_Float(
        A=&RandomRangeFloat(Minimum=$User.GlassColorVar, Maximum=1)
        B=&FloatFromCurve(FloatCurve=DataInterface() {propertyValues {Curve {keys [<at 0 1.5>,<at 0.1 1>,<at 0.9 1>,<at 1 1.5>]}}})
       ))
      False Vector=&VectorFromFloat(Value=&RandomRangeFloat(
        Minimum=&FloatFromCurve(FloatCurve=DataInterface() {propertyValues {Curve {keys [<at 0 30>,<at 0.3 1>]}}}, Scale Curve=0.5)
        Maximum=&FloatFromCurve(FloatCurve=DataInterface() {propertyValues {Curve {keys [<at 0 15>,<at 0.3 1>]}}}, Scale Curve=0.6)
       ))
     )
     Vector Group 5:Vector3f=<0 0 0>
    )
    ScaleA=true
    ScaleRGB=true
   )
   &SolveForcesAndVelocity(Clamp Velocity=false, Limit Acceleration=false)
   &ShapeLocation(
    MaskSpawnGroup=false
    Offset=&MaskVectorBySpawnGroup(
     Masked Vector=&MakeVector(
      X=&Sine(
       Normalized Angle=$Emitter.Age
       Period=&RandomRangeFloat(Minimum=2, Maximum=3)
       Scale=&RandomRangeFloat(Minimum=-40, Maximum=40)
       Bias=&RandomRangeFloat(Minimum=0, Maximum=1)
      )
      Y=&Sine(
       Normalized Angle=$Emitter.Age
       Period=&RandomRangeFloat(Minimum=2, Maximum=3)
       Scale=&RandomRangeFloat(Minimum=-40, Maximum=40)
       Bias=&RandomRangeFloat(Minimum=0, Maximum=1)
      )
      Z=&Sine(
       Normalized Angle=$Emitter.Age
       Period=&RandomRangeFloat(Minimum=2, Maximum=3)
       Scale=&RandomRangeFloat(Minimum=-40, Maximum=40)
       Bias=&RandomRangeFloat(Minimum=0, Maximum=1)
      )
     )
     Passthrough Vector=<0 0 0>
     Spawn Group Mask=3
    )
    Offset Mode=::NewEnumerator0
    Override Seed=false
    Shape Origin=$Particles.Initial.Position
    Sphere Radius=0
    UseOwnerScale=true
   ) {module=ShapeLocation001}
  ]
  rendererClasses [&NiagaraSpriteRendererProperties, &NiagaraMeshRendererProperties]
  renderers [
   {rendererClass=&NiagaraSpriteRendererProperties}
   {rendererClass=&NiagaraSpriteRendererProperties, rendererIndex=1}
   {rendererClass=&NiagaraSpriteRendererProperties, rendererIndex=2}
   {rendererClass=&NiagaraSpriteRendererProperties, rendererIndex=3}
   {rendererClass=&NiagaraMeshRendererProperties, rendererIndex=4}
   {rendererClass=&NiagaraSpriteRendererProperties, rendererIndex=5}
  ]
 }
 TriangleExplosion002 {
  emitterUpdateScript [
   &EmitterState()
   &VC_FilteredSocketCount() {enabled=false}
   &SpawnRate(SpawnRate=50, Use Spawn Probability=false) {enabled=false}
  ]
  particleSpawnScript [
   &VC_PaintedVariantsOverride() {enabled=false}
   &InitializeParticle(
    Color Mode=::NewEnumerator0
    Lifetime=&RandomRangeFloat(Minimum=0.2, Maximum=0.3)
    Sprite Rotation Mode=::NewEnumerator2
    Sprite Size Mode=::NewEnumerator3
    Uniform Sprite Size=&Multiply_Float(A=0.4, B=&RandomRangeFloat(Minimum=40, Maximum=60))
    UsePositionOffset=false
    Write Material Random=true
   )
   &PartitionParticles(Number Of Partitions=$Emitter.PortCount) {enabled=false}
   &SkeletalMeshLocation(
    Bone / Socket Index=$Particles.PartitionParticles.Partition
    Bone / Socket Sampling Mode=::NewEnumerator5
    MaskSpawnGroup=false
    Mesh Sampling Type=::NewEnumerator6
    Skeletal Mesh=$System.SkeletalMesh
   ) {enabled=false}
   &AddVelocity(Velocity=&RandomRangeVector(Minimum=<1000 1000 1000>, Maximum=<-1000 -1000 -1000>))
   &Color(Color=&MakeCustomLinearColorFromBool(
     Bool=$User.bIsPainted
     True Color=$User.PaintedColor
     False Color=&SelectLinearColorFromArray(
      Array Sampling Mode=::NewEnumerator0
      Direct Array Index=?
      Color Selection Array=DataInterface() {propertyValues {ColorData [<rgba 1 0.01 1 1>, <rgba 0.01 1 1 1>, <rgba 1 0.5 0.01 1>, <rgba 0.2 0.01 1 1>]}}
      Random Seed=0
     )
    ))
  ]
  particleUpdateScript [
   &ParticleState()
   &ScaleSpriteSize(Uniform Curve Sprite Scale=DataInterface() {propertyValues {Curve {keys [<at 0 0.55 RCIM_Cubic>, {arriveTangent=-0.55, interpMode=RCIM_Cubic, leaveTangent=-0.55, time=0.1, value=1}, <at 1 0 RCIM_Cubic>]}}})
   &ScaleColor(
    Scale RGB=&MakeCustomVectorFromBool(
     Bool=$User.bIsPainted
     True Vector=&VectorFromFloat(Value=&RandomRangeFloat(Minimum=$User.ParticleColorVar, Maximum=1))
     False Vector=&Multiply_VectorByFloat(
      Vector=&VectorFromFloat(Value=&RandomRangeFloat(Minimum=1, Maximum=1))
      Float=0.8
     )
    )
    ScaleA=true
    ScaleRGB=true
   )
   &Drag(Drag=5, Use Linear Drag=true, Use Rotational Drag=true)
   &SolveForcesAndVelocity(Clamp Velocity=false, Limit Acceleration=false)
  ]
  rendererClasses [&NiagaraSpriteRendererProperties]
  renderers [{rendererClass=&NiagaraSpriteRendererProperties}]
 }
}
summary {
 systemName=NS_Boost_HoloGlass_MP
 emitters [
  {emitterName=FakeGlow, rendererClasses [&NiagaraSpriteRendererProperties]}
  {
   emitterName=Sprites-LocalSpace
   rendererClasses [&NiagaraSpriteRendererProperties, &NiagaraMeshRendererProperties]
  }
  {emitterName=Sprites, rendererClasses [&NiagaraMeshRendererProperties, &NiagaraSpriteRendererProperties]}
  {emitterName=TriangleExplosion002, rendererClasses [&NiagaraSpriteRendererProperties]}
 ]
 userVariables [
  User.bIsFrontendPreview {type=NiagaraBool}
  User.bIsPainted {type=NiagaraBool}
  User.ColorVariantID {type=NiagaraInt32}
  User.CoreColor {type=LinearColor, description="", defaultValue {struct=LinearColor, value=<rgba 1 1 1 1>}}
  User.DTCarOffset {type=Vector3f, description="", defaultValue {struct=Vector3f, value=<0 0 0>}}
  User.GlassColor {type=LinearColor, description="", defaultValue {struct=LinearColor, value=<rgba 1 1 1 1>}}
  User.GlassColorVar {type=NiagaraFloat, defaultValue=1, description=""}
  User.GlowColor {type=LinearColor, description="", defaultValue {struct=LinearColor, value=<rgba 1 1 1 1>}}
  User.MaxGlassColor {type=NiagaraFloat, defaultValue=5, description=""}
  User.MinimumGlassColor {type=NiagaraFloat, defaultValue=0.5, description=""}
  User.PaintedColor {type=LinearColor, description="", defaultValue {struct=LinearColor, value=<rgba 1 1 1 1>}}
  User.PaintedScalar {type=NiagaraFloat, defaultValue=1, description=""}
  User.ParticleColorVar {type=NiagaraFloat, defaultValue=1, description=""}
  User.PortCount {type=NiagaraInt32, defaultValue=1}
  User.SingleBoostPos {type=NiagaraPosition, description="", defaultValue {}}
 ]
}
