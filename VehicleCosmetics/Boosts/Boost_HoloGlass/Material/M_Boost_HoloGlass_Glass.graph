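# Translucent unlit glass material for the HoloGlass vehicle boost effect, usable on Niagara mesh particles.
# Panning samples of a hologram texture at different tilings and speeds are summed with a border mask.
# Static switch "A" selects border alpha or red; the Border dynamic parameter's If returns red in every branch.
# Particle color tints the sum before MF_FX_EmissiveAdjustments, while particle alpha scales a saturated copy.
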Material /VehicleCosmetics/Boosts/Boost_HoloGlass/Material/M_Boost_HoloGlass_Glass
graph {
 properties {
  bAutomaticallySetUsageInEditor=false
  bUsedWithNiagaraMeshParticles=true
  blendMode=BLEND_Translucent
  shadingModel=MSM_Unlit
  usedShadingModels=Unlit
 }
 nodes [
  n1 = TextureSample() {texture=/DealershipSharedLibrary/FX/Textures/TexturePackage/T_VCVFX_HoloGlass_HoloBorder}
  n2 = Constant() {r=0.5}
  MaterialExpressionMeshEmitterDynamicParameter_0 = DynamicParameter() {paramNames [Border,Param2,Param3,Param4]}
  n3 = TextureSample(UVs=Panner(Coordinate=TextureCoordinate() {uTiling=0.9}) {speedX=0.3}) {texture=/DealershipSharedLibrary/FX/Textures/TexturePackage/T_VCVFX_HoloGlass_Hologram}
  MaterialExpressionConstantClamp_2 = Saturate(None=@n3.R)
  n4 = TextureCoordinate() {uTiling=0.25}
  n5 = TextureCoordinate() {uTiling=0.5}
  n6 = Add(
   A=TextureSample(UVs=Panner(Coordinate=@n4) {speedX=0.05}) {texture=&T_VCVFX_HoloGlass_Hologram}.G
   B=TextureSample(UVs=Panner(Coordinate=@n5) {speedX=0.02}) {texture=&T_VCVFX_HoloGlass_Hologram}.G
  )
  n7 = Constant() {r=1}
  n8 = Add(
   A=StaticSwitchParameter(True=@n1.A, False=If(A=@n2, B=@MaterialExpressionMeshEmitterDynamicParameter_0.Border, "A > B"=@n1.R, "A == B"=@n1.R, "A < B"=@n1.R)) {parameterName=A}
   B=Add(A=Add(A=@MaterialExpressionConstantClamp_2, B=@n3.B), B=Multiply(A=@n6, B=@n7))
  )
  MaterialExpressionMeshEmitterVertexColor_0 = ParticleColor()
  MaterialExpressionConstantClamp_1 = Saturate(None=Multiply(A=@n8, B=@MaterialExpressionMeshEmitterVertexColor_0.A))
  MaterialFunctionCall(Emissive=Multiply(A=@n8, B=@MaterialExpressionMeshEmitterVertexColor_0.RGB)) {materialFunction=&MF_FX_EmissiveAdjustments}
 ]
}
