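# Translucent unlit Niagara sprite material for the triangle effect of the HoloGlass vehicle boost cosmetic.
# Particle color RGB passes through MF_FX_EmissiveAdjustments, while particle alpha multiplies a triangle texture channel and is saturated.
# Switch "A" selects texture alpha or, by default, an If on scalar "RGB" whose branches all return blue.
# The "3d" switches choose between two identical ParticleColor nodes, so they change nothing.
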Material /VehicleCosmetics/Boosts/Boost_HoloGlass/Material/M_Boost_HoloGlass_Triagle
graph {
 properties {
  bAutomaticallySetUsageInEditor=false
  bUsedWithNiagaraSprites=true
  blendMode=BLEND_Translucent
  shadingModel=MSM_Unlit
  usedShadingModels=Unlit
 }
 nodes [
  n1 = TextureSampleParameter2D() {parameterName=RGBA, texture=/DealershipSharedLibrary/FX/Textures/TexturePackage/T_VCVFX_HoloGlass_Triangle}
  n2 = If(A=ScalarParameter() {parameterName=RGB}, B=Constant() {r=0.5}, "A > B"=@n1.B, "A == B"=@n1.B, "A < B"=@n1.B)
  MaterialExpressionMeshEmitterVertexColor_3 = ParticleColor()
  MaterialExpressionVertexColor_2 = ParticleColor()
  MaterialExpressionConstantClamp_4 = Saturate(None=Multiply(
    A=StaticSwitchParameter(True=@n2, False=@n1.A) {defaultValue=true, parameterName=A}
    B=StaticSwitchParameter(True=@MaterialExpressionMeshEmitterVertexColor_3.A, False=@MaterialExpressionVertexColor_2.A) {parameterName=3d}
   ))
  MaterialFunctionCall(Emissive=StaticSwitchParameter(True=@MaterialExpressionMeshEmitterVertexColor_3.RGB, False=@MaterialExpressionVertexColor_2.RGB) {parameterName=3d}) {materialFunction=&MF_FX_EmissiveAdjustments}
 ]
}
